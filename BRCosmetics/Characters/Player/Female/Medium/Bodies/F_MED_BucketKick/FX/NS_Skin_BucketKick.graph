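# Cosmetic Niagara effect for the BucketKick female medium body skin, attaching animated meshes and flowing particles to the character's skeleton.
# Arms bursts ten meshes pinned by per-group arrays, drawn by two renderers, with smoothed position and intensity passed as a particle vector.
# A BoneFlow emitter continuously moves pulsing, fading sprite/mesh particles along socket-based splines.
# Both emitters set VisibilityTag from the hand-hidden user flag, selectively by group for Arms.
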NiagaraSystem /BRCosmetics/Characters/Player/Female/Medium/Bodies/F_MED_BucketKick/FX/NS_Skin_BucketKick
dependencies {
 usedDataInterfaces [
  &NiagaraDataInterfaceArrayFloat3
  &NiagaraDataInterfaceArrayInt32
  &NiagaraDataInterfaceArrayFloat
  &NiagaraDataInterfaceCurve
  &NiagaraDataInterfaceSkeletalMesh
 ]
 usedDynamicInputs [
  &MakeCustomFloatFromBool
  &SetBoolByFloatComparison
  &VectorLength
  &Lerp_Float
  &ClampFloat
 ]
 usedModules [
  &self:SystemSpawnScript.SystemScriptSource.SystemScriptGraph.NiagaraNodeAssignment_3.SetVariables_#1
  &self:SystemSpawnScript.SystemScriptSource.SystemScriptGraph.NiagaraNodeAssignment_2.SetVariables_#2
  &SystemState
  &self:SystemSpawnScript.SystemScriptSource.SystemScriptGraph.NiagaraNodeAssignment_0.SetVariables_#3
 ]
}
emitters {
 Arms {
  emitterUpdateScript [
   &EmitterState(Life Cycle Mode=::NewEnumerator1, Loop Behavior=::NewEnumerator1, Loop Duration Mode=::NewEnumerator1, UseLoopDelay=false)
   &SpawnBurst_Instantaneous(Spawn Count=10, Use Spawn Probability=false, UseLoopCountLimit=false)
  ]
  particleSpawnScript [
   &FN_SetSpawnGroup(Number of Spawn Groups=10)
   &InitializeParticle(
    Color=<rgba 1 0.2795 0.653 1>
    Mesh Scale=&SelectVectorFromArray(
     Direct Array Index=$Particles.SpawnGroup
     Vector Selection Array=DataInterface() {propertyValues {FloatData [
        <1.2 1.2 1.2>
        <1.5 1.5 1.5>
        <0.8 -0.8 0.8>
        <1.2 -1.2 1.2>
        <1.2 1.2 1.2>
        <1.3 1.3 1.3>
        <1.1 1.1 1.1>
        <0.7 0.7 0.7>
        <1 -1 1>
        <0.8 -0.8 0.8>
       ]}}
    )
    Mesh Scale Mode=::NewEnumerator4
    UsePositionOffset=false
    Write Material Random=true
   )
   &InitialMeshOrientation(
    Mesh Axis To Orient=<0 0 1>
    Mesh Orientation Mode=::NewEnumerator2
    Orientation Vector=&SelectVectorFromArray(
     Direct Array Index=$Particles.SpawnGroup
     Vector Selection Array=DataInterface() {propertyValues {FloatData [<-1 1 -0.4>, <-1 0.3 0>, <-1 0.3 0.5>, <-1 0.6 -0.3>, <1 -1 0>, <1 0.5 -1>, <-1 2 0>, <-0.5 -0.2 0.2>, <-0.8 -1 -0.5>, <-10 0.5 1>]}}
    )
    Use Rotation Vector=false
   )
   /SetVariables(
    Particles.R2_Color:LinearColor=<rgba 1 0 0.4853 1>
    Particles.R2_Scale:Vector3f=&Multiply_Vector(A=$Particles.Scale, B=<1.25 1.25 1.25>)
   ) {isSetParameters=true}
   /SetVariables(Particles.RandU:NiagaraFloat=&RandomRangeFloat(Minimum=0, Maximum=1)) {isSetParameters=true}
  ]
  particleUpdateScript [
   &ParticleState(Kill Particles When Lifetime Has Elapsed=false, Let Infinitely Lived Particles Die When Emitter Deactivates=true)
   &SkeletalMeshLocation(
    Bone / Socket Index=&SelectIntFromArray(Direct Array Index=$Particles.SpawnGroup, Int Selection Array=DataInterface() {propertyValues {IntData [0,1,1,1,4,5,6,6,6]}})
    MaskSpawnGroup=false
    Mesh Sampling Type=::NewEnumerator5
    Sampled Position Offset=&SelectVectorFromArray(
     Direct Array Index=$Particles.SpawnGroup
     Vector Selection Array=DataInterface() {propertyValues {FloatData [<2.5 0 0.25>, <-3 0 0>, <3 4 0>, <0 4 0>, <-1.5 0 -4>, <-7 0 -0.5>, <1 2 3>, <0 -4 3>, <4 -2 5>, <4 1 5>]}}
    )
    Skeletal Mesh=$System.Skeletal Mesh
    Socket Sampling Mode=::NewEnumerator5
   )
   &InterpolateOverTime(
    Rate Of Change:NiagaraFloat=$System.Interpolate Rate
    Target Vector Value:NiagaraPosition=$Particles.Position
    Variable Type=::NewEnumerator2
   )
   &DynamicMaterialParameters(
    Index 0 Param 1=&SelectFloatFromArray(
     Direct Array Index=$Particles.SpawnGroup
     Float Selection Array=DataInterface() {propertyValues {FloatData [2,2,1.8,2.1,2.4,2.3,1.5,1.1,1.5]}}
    )
    Index 0 Param 2=&Multiply_Float(A=$Engine.Time, B=1)
    Index 0 Param 3=&SelectFloatFromArray(
     Direct Array Index=$Particles.SpawnGroup
     Float Selection Array=DataInterface() {propertyValues {FloatData [0.25,0.3,0.22,0.3,0.32,0.3,0.25,0.2,0.25]}}
    )
    Index 0 Param 4=$Particles.RandU
    Index 2 Param 1=1
    Module.Index0Param0Write=true
    Module.Index0Param1Write=true
    Module.Index0Param2Write=true
    Module.Index0Param3Write=true
    Module.Index2Param0Write=true
    Module.Index2Param1Write=true
    Module.Index2Param2Write=true
    Module.Index2Param3Write=true
    Write Parameter Index 2=true
   )
   /SetVariables(Particles.Interpolate DY:Vector4f=&MakeFloat4FromVectorAndFloat("Vector (RGB)"=$Particles.InterpolateOverTime.InterpolateOverTime_V3_.Moving Average, "Float (Alpha)"=$System.Interpolate Internsity)) {isSetParameters=true}
   /SetVariables(Particles.R2_DynMatParm:Vector4f=&MakeVector4(
     X=&Multiply_Float(
      A=1.25
      B=&SelectFloatFromArray(
       Direct Array Index=$Particles.SpawnGroup
       Float Selection Array=DataInterface() {propertyValues {FloatData [2,2,1.8,2.1,2.4,2.3,1.5,1.1,1.5]}}
      )
     )
     Y=&Multiply_Float(A=$Engine.Time, B=1)
     Z=&SelectFloatFromArray(
      Direct Array Index=$Particles.SpawnGroup
      Float Selection Array=DataInterface() {propertyValues {FloatData [0.25,0.3,0.22,0.3,0.32,0.3,0.25,0.2,0.25]}}
     )
     W=$Particles.RandU
    )) {isSetParameters=true}
   /SetVariables() {isSetParameters=true}
   /SetVariables(Particles.VisibilityTag=&MakeCustomIntFromBool(
     Bool=$User.bisHandHidden
     True INT=&MakeCustomIntFromBool(
      Bool=&Boolean_OrOperation(
       Bool1=&SetBoolByIntComparison(A=$Particles.SpawnGroup, B=2, Comparison Type=::NewEnumerator4)
       Bool2=&SetBoolByIntComparison(A=$Particles.SpawnGroup, B=5, Comparison Type=::NewEnumerator2)
      )
      True INT=0
      False INT=1
     )
     False INT=0
    )) {isSetParameters=true}
  ]
  rendererClasses [&NiagaraMeshRendererProperties]
  renderers [
   {rendererClass=&NiagaraMeshRendererProperties, rendererIndex=0}
   {rendererClass=&NiagaraMeshRendererProperties, rendererIndex=1}
  ]
 }
 FN_BoneFlow_Sprite_BETA {
  emitterUpdateScript [
   &EmitterState(Life Cycle Mode=::NewEnumerator1, Loop Duration=1, UseLoopDelay=false)
   &SpawnRate(SpawnRate=4, Use Spawn Probability=false)
  ]
  particleSpawnScript [
   &FN_SetSpawnGroup(Number of Spawn Groups=2)
   &InitializeParticle(
    Lifetime Max=2.3
    Lifetime Min=2
    Lifetime Mode=::NewEnumerator1
    Mesh Scale Mode=::NewEnumerator1
    Mesh Uniform Scale Max=1.5
    Mesh Uniform Scale Min=0.9
    Sprite Size Mode=::NewEnumerator3
    UsePositionOffset=false
    Write Material Random=true
   )
   /SetVariables(Particles.RandU:NiagaraFloat=&RandomRangeFloat(Minimum=0, Maximum=1)) {isSetParameters=true}
   &InitialMeshOrientation(Mesh Orientation Mode=::NewEnumerator2, Use Rotation Vector=false)
  ]
  particleUpdateScript [
   &ParticleState()
   &ScaleMeshSize(Scale Factor=&VectorFromFloat(Value=&FloatFromCurve(FloatCurve=DataInterface() {propertyValues {Curve {keys [
          {interpMode=RCIM_Cubic, tangentMode=RCTM_User, time=0, value=0.5}
          {interpMode=RCIM_Cubic, tangentMode=RCTM_User, time=0.5, value=1}
          {tangentMode=RCTM_User, time=1, value=0.5}
         ]}}})))
   /Game/Effects/Niagara/Modules/BoneFlow/FN_BoneFlow_v4/FN_BoneFlow(
    Angle:NiagaraFloat=&RandomRangeFloat(Minimum=0, Maximum=0.5)
    BoneID:NiagaraInt32=$Particles.SpawnGroup
    Height:NiagaraFloat=&FloatFromCurve(FloatCurve=DataInterface() {propertyValues {Curve {keys [{tangentMode=RCTM_User, time=0, value=0.4}, {tangentMode=RCTM_User, time=1, value=1}]}}})
    Radius=1.8
    SkeletalMesh:NiagaraDataInterfaceSkeletalMesh=$System.Skeletal Mesh
    Use Sockets Instead=true
    Write Position Attribute=true
   ) {module=FN_BoneFlow001}
   &UpdateMeshOrientation(
    Bank Rate=10
    Banking Angle Min / Max:Vector2f=<0 0>
    "Foward Vector (Mesh Space)":Vector3f=$Output.FN_BoneFlow001.SplineBinormal
    Initial Forward Vector:Vector3f=&MaskVectorBySpawnGroup(Masked Vector=<1 0 0>, Passthrough Vector=<1 0 0.5>)
    Orientation Method=::NewEnumerator3
    "Right Vector (Mesh Space)":Vector3f=$Output.FN_BoneFlow001.SplineTangent
    Rotation Decay Rate In Degrees=0
    Up Vector:Vector3f=&NormalizeVector(
     Vector To Normalize=&SubtractVectorFromPosition(Position=$Particles.Position, Vector=$Output.FN_BoneFlow001.PositionOnSpline)
     Safe Normalize=true
     Safe Fallback Threshold=0
     Safe Fallback Vector=<0 0 0.001>
    )
    Upright Roll Correction Speed=0
   )
   &DynamicMaterialParameters(
    Index 0 Param 1=2
    Index 0 Param 2=&Multiply_Float(A=$Engine.Time, B=1)
    Index 0 Param 3=0.25
    Index 0 Param 4=$Particles.RandU
    Index 2 Param 1=1
    Module.Index0Param0Write=true
    Module.Index0Param1Write=true
    Module.Index0Param2Write=true
    Module.Index0Param3Write=true
    Module.Index2Param0Write=true
    Module.Index2Param1Write=true
    Module.Index2Param2Write=true
    Module.Index2Param3Write=true
    Write Parameter Index 2=true
   )
   &ScaleColor(
    Scale Alpha=&FloatFromCurve(FloatCurve=DataInterface() {propertyValues {Curve {keys [
         {interpMode=RCIM_Cubic, tangentMode=RCTM_User, time=0, value=0}
         {interpMode=RCIM_Cubic, tangentMode=RCTM_User, time=0.5, value=1}
         {tangentMode=RCTM_User, time=1, value=0}
        ]}}})
    ScaleA=true
    ScaleRGB=true
   )
   /SetVariables(Particles.VisibilityTag=&MakeCustomIntFromBool(Bool=$User.bisHandHidden, True INT=1, False INT=0)) {isSetParameters=true}
  ]
  rendererClasses [&NiagaraSpriteRendererProperties, &NiagaraMeshRendererProperties]
  renderers [
   {rendererClass=&NiagaraSpriteRendererProperties}
   {rendererClass=&NiagaraMeshRendererProperties, rendererIndex=1}
  ]
 }
}
summary {
 systemName=NS_Skin_BucketKick
 emitters [
  {emitterName=Arms, rendererClasses [&NiagaraMeshRendererProperties]}
  {
   emitterName=FN_BoneFlow_Sprite_BETA
   rendererClasses [&NiagaraSpriteRendererProperties, &NiagaraMeshRendererProperties]
  }
 ]
 userVariables [
  User.bIsFrontEnd {type=NiagaraBool}
  User.bIsFrontEndPreview {type=NiagaraBool}
  User.bisHandHidden {type=NiagaraBool}
 ]
}
