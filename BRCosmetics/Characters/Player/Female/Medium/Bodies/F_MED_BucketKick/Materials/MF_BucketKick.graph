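# Material function for the BucketKick character body that layers an animated gradient emissive effect with darkened base color onto input material attributes, apparently blended by a channel-selected mask.
# Fresnel or texture values drive the gradient lookup, with distortion built from flow-warped world-aligned noise on pre-skinned positions and a sine wave.
# A static switch adds roughness, metallic, specular and a swirling world position offset from radially panned noise, suppressed inside a sphere mask.
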MaterialFunction /BRCosmetics/Characters/Player/Female/Medium/Bodies/F_MED_BucketKick/Materials/MF_BucketKick
graph {nodes [
  n1 = NamedRerouteUsage()
  n2 = SetMaterialAttributes!(
   MaterialAttributes=@n1
   EmissiveColor=NamedRerouteUsage()
   BaseColor=Multiply(
    A=GetMaterialAttributes(None=@n1) {attributeGetTypes [#1]}.BaseColor
    B=ScalarParameter() {defaultValue=0.4, group=BucketKick, parameterName=BaseColorMulitpler}
   )
  ) {attributeSetTypes [#2, #1]}
  n3 = SetMaterialAttributes!(
   MaterialAttributes=@n2
   Roughness=ScalarParameter() {defaultValue=0.5, group=BucketKick, parameterName=Roughness, sortPriority=101}
   Metallic=ScalarParameter() {group=BucketKick, parameterName=Metallic, sortPriority=102}
   Specular=ScalarParameter() {defaultValue=0.4, group=BucketKick, parameterName=Specular, sortPriority=103}
   WorldPositionOffset=NamedRerouteUsage()
  ) {attributeSetTypes [
    #3
    #4
    #5
    #6
   ]}
  FunctionOutput(None=BlendMaterialAttributes(
    A=@n1
    B=StaticSwitchParameter(True=@n3, False=@n2) {group=BucketKick, parameterName=OverrideAdditionalMaterialParams, sortPriority=100}
    Alpha=NamedRerouteUsage()
   ))
  n4 = TextureSample(UVs=MaterialFunctionCall(X=NamedRerouteUsage(), Y=Constant() {r=0.5}) {materialFunction=&MakeFloat2}.Result) {
   samplerSource=SSM_Clamp_WorldGroupSettings
   texture=/BRCosmetics/Characters/Player/Female/Medium/Bodies/F_MED_BucketKick/Textures/T_BucketKick_Gradient
  }
  n5 = ScalarParameter() {defaultValue=1.2, group=BucketKick, parameterName=GradientExponent, sortPriority=50}
  NamedRerouteDeclaration(None=Multiply(A=Power(Base=@n4.RGB, Exp=@n5), B=ScalarParameter() {defaultValue=1.2, group=BucketKick, parameterName=EmissiveStrength, sortPriority=51})) {name=EmissiveOUT, nodeColor=<rgba 1 0.0938 0 1>}
  n6 = TextureObject() {samplerType=SAMPLERTYPE_Normal, texture=/Game/Global/Textures/Default/Flowmaps/T_SwirlFlow_02_N}
  n7 = ScalarParameter() {defaultValue=40, group=BucketKick, parameterName=DistortionFlowTextureSize}
  n8 = PreSkinnedPosition()
  n9 = Add(
   A=NamedRerouteDeclaration(None=VertexInterpolator(VS=@n8).PS) {name=PreskinnedPos, nodeColor=<rgba 0 0.9766 1 1>}
   B=Multiply(A=Time(), B=VectorParameter() {group=BucketKick, parameterName=DistortionFlowSpeed, defaultValue=<rgba 0 0 -4 1>}.RGB)
  )
  n10 = NamedRerouteDeclaration(None=VertexInterpolator(VS=PreSkinnedNormal()).PS) {name=PreskinnedNormal, nodeColor=<rgba 1 0 0.6562 1>}
  n11 = NamedRerouteUsage()
  NamedRerouteDeclaration(None=Multiply(
    A=MaterialFunctionCall(TextureObject=@n6, TextureSize=@n7, WorldPosition=@n9, World Space Normal=@n10, ProjectionTransitionContrast=@n11) {materialFunction=&WorldAlignedTexture}.XYZ Texture
    B=ScalarParameter() {defaultValue=3, group=BucketKick, parameterName=DistortionFlowStrength, sortPriority=20}
   )) {name=Flow, nodeColor=<rgba 0.3594 0 1 1>}
  n12 = ConstantBiasScale(None=NamedRerouteUsage()) {bias=-0.5, scale=2}
  n13 = ScalarParameter() {defaultValue=0.1, group=BucketKick, parameterName=TextureDistortionStrength, sortPriority=42}
  n14 = TextureSampleParameter2D(UVs=Add(A=TextureCoordinate(), B=Multiply(A=@n12, B=@n13))) {
   group=BucketKick
   parameterName=GradientUVTexture
   samplerType=SAMPLERTYPE_Masks
   sortPriority=41
   texture=&T_BlackMask
  }
  NamedRerouteDeclaration(None=Power(Base=@n14.R, Exp=ScalarParameter() {defaultValue=1, group=BucketKick, parameterName=TextureGradientUVExponent, sortPriority=42})) {name=TextureGradientUVs, nodeColor=<rgba 0.3594 1 0 1>}
  n15 = LocalPosition()
  n16 = ComponentMask(None=@n15.XYZ) {b=true, g=true}
  n17 = ComponentMask(None=LocalPosition().XYZ) {b=true, g=true}
  n18 = VectorParameter() {group=BucketKick, parameterName=PositionObjectCenter, sortPriority=110, defaultValue=<rgba 5 62 0 1>}
  n19 = ComponentMask(None=@n18.RGB) {g=true, r=true}
  n20 = Multiply(
   A=Add(
    A=NamedRerouteDeclaration(None=@n17) {name=LocalPosYZ, nodeColor=<rgba 1 0.9375 0 1>}
    B=NamedRerouteDeclaration(None=@n19) {name=WPOLocalOrigin, nodeColor=<rgba 0 0.7422 1 1>}
   )
   B=ScalarParameter() {defaultValue=0.05, group=BucketKick, parameterName=WPOTextureSize, sortPriority=111}
  ) {constB=0.025}
  n21 = VectorParameter() {group=BucketKick, parameterName=WPOTextureSpeed, sortPriority=112, defaultValue=<rgba -0.1 -0.35 0 1>}
  n22 = TextureSample!(
   UVs=Panner(
    Coordinate=MaterialFunctionCall(Vector or UVs=@n20) {materialFunction=&VectorToRadialValue}.Radial Coordinates
    Speed=ComponentMask(None=@n21.RGB) {g=true, r=true}
   ) {speedX=-0.1, speedY=-0.35}
   Tex
   Level
   Apply View MipBias
  ) {
   mipValueMode=TMVM_MipLevel
   samplerType=SAMPLERTYPE_LinearColor
   texture=/BRCosmetics/Characters/Player/Female/Medium/Bodies/F_MED_BucketKick/Textures/T_BucketKick_SwirlNoise
  }
  n23 = Multiply(A=@n22.R) {constB=2}
  n24 = Multiply(
   A=NamedRerouteDeclaration(None=Subtract(A=@n23)) {name=WPOTexture, nodeColor=<rgba 0.0078 1 0 1>}
   B=ScalarParameter() {defaultValue=2, group=BucketKick, parameterName=WPODistortionIntensity, sortPriority=113}
  ) {constB=2}
  n25 = Subtract(A=AppendVector(A=ComponentMask(None=@n15.XYZ) {r=true}, B=Add(A=@n16, B=@n24)), B=LocalPosition().XYZ)
  n26 = Saturate(None=SphereMask(
    A=NamedRerouteUsage()
    B=NamedRerouteUsage()
    Radius=ScalarParameter() {defaultValue=25, group=BucketKick, parameterName=WPOSphereMaskRadius, sortPriority=114}
    Hardness=ScalarParameter() {defaultValue=0.25, group=BucketKick, parameterName=WPOSphereMaskHardness, sortPriority=115}
   ) {attenuationRadius=10})
  NamedRerouteDeclaration(None=Multiply(A=Transform(None=@n25) {transformSourceType=TRANSFORMSOURCE_Local}, B=OneMinus(None=@n26))) {name=WPO, nodeColor=<rgba 0.5703 1 0 1>}
  n27 = TextureObject() {
   samplerType=SAMPLERTYPE_LinearColor
   texture=&T_BucketKick_SwirlNoise
  }
  n28 = ScalarParameter() {defaultValue=20, group=BucketKick, parameterName=DistortionNoiseTextureSize, sortPriority=12}
  n29 = Time()
  n30 = VectorParameter() {group=BucketKick, parameterName=DistortionNoiseSpeed, sortPriority=12, defaultValue=<rgba 1 0 -3 1>}
  n31 = Add(A=Add(A=NamedRerouteUsage(), B=Multiply(A=@n29, B=@n30.RGB)), B=NamedRerouteUsage())
  n32 = NamedRerouteUsage()
  n33 = NamedRerouteDeclaration(None=ScalarParameter() {defaultValue=1, group=BucketKick, parameterName=DistortionNoiseTransitionContrast, sortPriority=13}) {name=TransitionContrast, nodeColor=<rgba 1 0.9844 0 1>}
  NamedRerouteDeclaration(None=ComponentMask(None=MaterialFunctionCall(TextureObject=@n27, TextureSize=@n28, WorldPosition=@n31, World Space Normal=@n32, ProjectionTransitionContrast=@n33) {materialFunction=&WorldAlignedTexture}.XYZ Texture) {g=true}) {name=DistortionNoiseSample, nodeColor=<rgba 1 0 0.5391 1>}
  NamedRerouteDeclaration(None=FunctionInput() {inputName=Material, inputType=FunctionInput_MaterialAttributes}) {name=IN_Material, nodeColor=<rgba 1 0.5391 0 1>}
  n34 = Power(Base=NamedRerouteUsage(), Exp=ScalarParameter() {defaultValue=0.8, group=BucketKick, parameterName=DistortionNoiseExponent, sortPriority=14})
  NamedRerouteDeclaration(None=Multiply(
    A=ConstantBiasScale(None=@n34) {bias=-0.5, scale=2}
    B=ScalarParameter() {defaultValue=0.2, group=BucketKick, parameterName=DistortionNoiseStrength, sortPriority=10}
   )) {name=Distortion, nodeColor=<rgba 1 0.7031 0 1>}
  n35 = NamedRerouteUsage()
  n36 = Add(
   A=Multiply(
    A=ComponentMask(None=@n35) {g=true}
    B=ScalarParameter() {defaultValue=5, group=BucketKick, parameterName=DistortionWaveFrequency, sortPriority=31}
   )
   B=Multiply(A=Time(), B=ScalarParameter() {defaultValue=0.5, group=BucketKick, parameterName=DistortionWaveSpeed})
  )
  NamedRerouteDeclaration(None=Multiply(A=Sine(None=@n36), B=ScalarParameter() {defaultValue=0.075, group=BucketKick, parameterName=DistortionWaveStrength, sortPriority=30})) {name=WaveDistortion, nodeColor=<rgba 0.3828 0 1 1>}
  n37 = NamedRerouteUsage()
  n38 = NamedRerouteUsage()
  n39 = CameraVectorWS()
  n40 = PixelNormalWS()
  n41 = Power(
   Base=Saturate(None=DotProduct(A=@n39, B=@n40))
   Exp=ScalarParameter() {defaultValue=1.5, group=BucketKick, parameterName=FresnelExponent, sortPriority=43}
  )
  n42 = StaticSwitchParameter(
   True=LinearInterpolate(A=@n38, B=@n41, Alpha=NamedRerouteUsage())
   False=StaticSwitchParameter(True=@n38, False=@n41) {group=BucketKick, parameterName=UseGradientTextureUV, sortPriority=40}
  ) {group=BucketKick, parameterName=UseTextureFresnelBlend, sortPriority=40}
  NamedRerouteDeclaration(None=Saturate(None=Add(A=@n37, B=@n42))) {name=GradientUVX, nodeColor=<rgba 0 1 0.0625 1>}
  NamedRerouteDeclaration(None=ChannelMaskParameter(None=FunctionInput() {bUsePreviewValueAsDefault=true, inputName=Mask, sortPriority=1}) {group=BucketKick, parameterName=MaskChannel, sortPriority=-10}.R) {name=Mask, nodeColor=<rgba 0.2188 0 1 1>}
  NamedRerouteDeclaration(None=@n14.G) {name=TextureGradientUVMask, nodeColor=<rgba 0.8047 1 0 1>}
  NamedRerouteDeclaration(None=StaticSwitchParameter(True=TextureCoordinate() {coordinateIndex=1}, False=TextureCoordinate()) {defaultValue=true, group=BucketKick, parameterName=UseUV2, sortPriority=0}) {name=UV, nodeColor=<rgba 1 0 0.5156 1>}
  VertexNormalWS()
 ]}
