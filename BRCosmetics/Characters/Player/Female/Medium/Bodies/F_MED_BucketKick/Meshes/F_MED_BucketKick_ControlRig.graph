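# Control Rig for the female medium BucketKick body, adding procedural secondary motion to outfit bones on top of a shared deform-rig function.
# Weighted parent constraints place shirt-bottom bones along the spine, the left kneepad and left skirt roots, and a waist-chain target that the chain aims at.
# Spherical pose readers on the left upper arm and both thighs weight corrective offsets for the shoulder pad, skirt and waist chain.
# Skirt rotation constraints are left unconnected.
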ControlRigBlueprint /BRCosmetics/Characters/Player/Female/Medium/Bodies/F_MED_BucketKick/Meshes/F_MED_BucketKick_ControlRig
graphs [
 RigVMFunctionLibrary {links [], nodes []}
 RigVMFunction_Sequence_1_1_ContainedGraph {nodes [
   n1 = RigVMFunctionEntryNode()
   n2 = &RigVMFunction_Sequence(ExecuteContext=@n1.ExecuteContext)
   n3 = &RigVMFunction_Sequence(ExecuteContext=@n2.B)
   RigVMFunctionReturnNode(A=@n2.A, B=@n3.A, C=@n3.B)
  ]}
 RigVMFunction_Sequence_1_ContainedGraph {nodes [
   n1 = RigVMFunctionEntryNode()
   RigVMFunction_Sequence = &RigVMFunction_Sequence(ExecuteContext=@n1.ExecuteContext)
   n2 = &RigVMFunction_Sequence(ExecuteContext=@RigVMFunction_Sequence.B)
   n3 = &RigVMFunction_Sequence(ExecuteContext=@n2.B)
   n4 = &RigVMFunction_Sequence(ExecuteContext=@n3.B)
   n5 = &RigVMFunction_Sequence(ExecuteContext=@n4.B)
   RigVMFunctionReturnNode(A=@RigVMFunction_Sequence.A, B=@n2.A, C=@n3.A, D=@n4.A, E=@n5.A, F=@n5.B)
  ]}
 RigVMModel {nodes [
   RigUnit_BeginExecution = &RigUnit_BeginExecution()
   RigVMFunction_Sequence = RigVMAggregateNode(ExecuteContext=@RigUnit_BeginExecution.ExecutePin)
   FN_DeformRig_FNC = RigVMFunctionReferenceNode(
    DeformRig_ Curve="(deformRig_l_upperarm_up_alpha_72_#1=False,deformRig_r_upperarm_up_alpha_73_#2=False,deformRig_l_upperarm_fr_alpha_74_#3=False,deformRig_r_upperarm_fr_alpha_75_CB51B8 ...(truncated)"
    DeformRig_Bool="(upperarm_1_#4=true,elbow_3_#5=true,wrist_7_#6=true,groin_8_#7=true,glute_17_#8=true,knee_9_40C408324CA261 ...(truncated)"
    DeformRig_Debug=(upperarm_debug_18_#4=False,wrist_debug_20_#6=False,lat_debug_26_#9=False,groinGlute_debug_27_#7=False)
    DeformRig_Value="(upperarm_up_18_#10=(Rotation=(X=0,Y=0,Z=0,W=1),Translation=(X=0,Y=0,Z=8),Scale3D=(X=0.7,Y=1,Z=1)),upperarm_fwd_fr_19_#11=(Rotation=(X=0,Y=0,Z=0,W=1),Translation=(X=0,Y=10,Z=0),Scale3D=(X=0.5,Y=1,Z= ...(truncated)"
    ExecuteContext=@RigVMFunction_Sequence.A
   )
   &RigUnit_ParentConstraint(
    AdvancedSettings=(InterpolationType=Average,RotationOrderForFilter=XZY)
    Child=(Type=Bone,Name="dyn_shirtbott_f_MSTR")
    Parents=((Item=(Type=Bone,Name="pelvis"),Weight=0.7075),(Item=(Type=Bone,Name="spine_01"),Weight=0.1228),(Item=(Type=Bone,Name="spine_02"),Weight=0.2627),(Item=(Type=Bone,Name="spine_03"),Weight=0.2008),(Item=(Type=Bone,Name="spine_04"),Weight=0.3666))
    Weight=1
    ExecutePin=@RigVMFunction_Sequence.B
   )
   > &RigUnit_ParentConstraint(
    AdvancedSettings=(InterpolationType=Average,RotationOrderForFilter=XZY)
    Child=(Type=Bone,Name="dyn_shirtbott_a_MSTR")
    Parents=((Item=(Type=Bone,Name="pelvis"),Weight=0.3095),(Item=(Type=Bone,Name="spine_01"),Weight=0.1822),(Item=(Type=Bone,Name="spine_02"),Weight=0.3202),(Item=(Type=Bone,Name="spine_03"),Weight=0.0585),(Item=(Type=Bone,Name="spine_04"),Weight=0.1296))
    Weight=1
   )
   > &RigUnit_ParentConstraint(
    AdvancedSettings=(InterpolationType=Average,RotationOrderForFilter=XZY)
    Child=(Type=Bone,Name="dyn_shirtbott_b_MSTR")
    Parents=((Item=(Type=Bone,Name="pelvis"),Weight=0.0652),(Item=(Type=Bone,Name="spine_01"),Weight=0.1272),(Item=(Type=Bone,Name="spine_02"),Weight=0.268),(Item=(Type=Bone,Name="spine_03"),Weight=0.1991),(Item=(Type=Bone,Name="spine_04"),Weight=0.3623))
    Weight=1
   )
   > &RigUnit_ParentConstraint(
    AdvancedSettings=(InterpolationType=Average,RotationOrderForFilter=XZY)
    Child=(Type=Bone,Name="dyn_shirtbott_c_MSTR")
    Parents=((Item=(Type=Bone,Name="spine_01"),Weight=0.0734),(Item=(Type=Bone,Name="spine_02"),Weight=0.1739),(Item=(Type=Bone,Name="spine_03"),Weight=0.2853),(Item=(Type=Bone,Name="spine_04"),Weight=0.4675))
    Weight=1
   )
   > &RigUnit_ParentConstraint(
    AdvancedSettings=(InterpolationType=Average,RotationOrderForFilter=XZY)
    Child=(Type=Bone,Name="dyn_shirtbott_d_MSTR")
    Parents=((Item=(Type=Bone,Name="spine_01"),Weight=0.1013),(Item=(Type=Bone,Name="spine_02"),Weight=0.2244),(Item=(Type=Bone,Name="spine_03"),Weight=0.3081),(Item=(Type=Bone,Name="spine_04"),Weight=0.3662))
    Weight=1
   )
   > &RigUnit_ParentConstraint(
    AdvancedSettings=(InterpolationType=Average,RotationOrderForFilter=XZY)
    Child=(Type=Bone,Name="dyn_shirtbott_e_MSTR")
    Parents=((Item=(Type=Bone,Name="pelvis"),Weight=0.2438),(Item=(Type=Bone,Name="spine_01"),Weight=0.2029),(Item=(Type=Bone,Name="spine_02"),Weight=0.3186),(Item=(Type=Bone,Name="spine_03"),Weight=0.1578),(Item=(Type=Bone,Name="spine_04"),Weight=0.0769))
    Weight=1
   )
   &RigUnit_ParentConstraint(
    AdvancedSettings=(InterpolationType=Average,RotationOrderForFilter=XZY)
    Child=(Type=Bone,Name="dyn_kneepad_l")
    Parents=((Item=(Type=Bone,Name="thigh_l"),Weight=0.25),(Item=(Type=Bone,Name="calf_l"),Weight=0.5),(Item=(Type=Bone,Name="deform_knee_out_l"),Weight=0.25))
    Weight=1
    ExecutePin=@RigVMFunction_Sequence.C
   )
   &RigUnit_ParentConstraint(
    AdvancedSettings=(InterpolationType=Average,RotationOrderForFilter=XZY)
    Child=(Type=Bone,Name="dyn_skirt_d_MSTR")
    Filter=(TranslationFilter=(bX=True,bY=True,bZ=True),RotationFilter=(bX=false,bY=false,bZ=false),ScaleFilter=(bX=false,bY=false,bZ=false))
    Parents=((Item=(Type=Bone,Name="pelvis"),Weight=1),(Item=(Type=Bone,Name="thigh_l"),Weight=0),(Item=(Type=Bone,Name="thigh_twist_01_l"),Weight=0),(Item=(Type=Bone,Name="deform_groin_l"),Weight=0))
    Weight=1
    ExecutePin=@RigVMFunction_Sequence.D
   )
   > &RigUnit_ParentConstraint(
    AdvancedSettings=(InterpolationType=Average,RotationOrderForFilter=XZY)
    Child=(Type=Bone,Name="dyn_skirt_h_MSTR")
    Filter=(TranslationFilter=(bX=True,bY=True,bZ=True),RotationFilter=(bX=false,bY=false,bZ=false),ScaleFilter=(bX=false,bY=false,bZ=false))
    Parents=((Item=(Type=Bone,Name="pelvis"),Weight=0.6253),(Item=(Type=Bone,Name="thigh_l"),Weight=0.2346),(Item=(Type=Bone,Name="thigh_twist_01_l"),Weight=0.1157),(Item=(Type=Bone,Name="deform_groin_l"),Weight=0.0244))
    Weight=1
   )
   > &RigUnit_ParentConstraint(
    AdvancedSettings=(InterpolationType=Average,RotationOrderForFilter=XZY)
    Child=(Type=Bone,Name="dyn_skirt_c_MSTR")
    Filter=(TranslationFilter=(bX=True,bY=True,bZ=True),RotationFilter=(bX=false,bY=false,bZ=false),ScaleFilter=(bX=false,bY=false,bZ=false))
    Parents=((Item=(Type=Bone,Name="pelvis"),Weight=0.653),(Item=(Type=Bone,Name="thigh_l"),Weight=0.3347),(Item=(Type=Bone,Name="thigh_twist_01_l"),Weight=0.0013),(Item=(Type=Bone,Name="deform_groin_l"),Weight=0.011))
    Weight=1
   )
   RigVMFunction_Sequence_1 = RigVMAggregateNode(ExecuteContext=@RigVMFunction_Sequence.E)
   n1 = &RigUnit_SphericalPoseReader(
    ActiveRegionScaleFactors=(PositiveWidth=1,NegativeWidth=1,PositiveHeight=1,NegativeHeight=1)
    ActiveRegionSize=0.1
    Debug=(bDrawDebug=True,bDraw2D=False,bDrawLocalAxes=False,DebugScale=25,DebugSegments=20,DebugThickness=0.25)
    DriverAxis=(X=1,Y=0,Z=0)
    DriverItem=(Type=Bone,Name="upperarm_twist_01_l")
    FalloffRegionScaleFactors=(PositiveWidth=1,NegativeWidth=1,PositiveHeight=1,NegativeHeight=1)
    FalloffSize=0.4
    FlipHeightScaling=False
    FlipWidthScaling=False
    OptionalParentItem=(Type=Bone,Name="spine_05")
    RotationOffset=(X=7.3845,Y=16.6767,Z=-28.9589)
    ExecutePin=@RigVMFunction_Sequence_1.A
   )
   > &RigUnit_ModifyTransforms(
    ItemToModify=((Item=(Type=Bone,Name="dyn_shoulderpad_l_MSTR"),Transform=(Rotation=(X=0,Y=0,Z=0,W=1),Translation=(X=7.6632,Y=0,Z=0),Scale3D=(X=1,Y=1,Z=1))))
    WeightMinimum=0
    Weight=@n1.OutputParam
   )
   n2 = &RigUnit_SphericalPoseReader(
    ActiveRegionScaleFactors=(PositiveWidth=0.8,NegativeWidth=0.3,PositiveHeight=0.3,NegativeHeight=0.3)
    ActiveRegionSize=0.4
    Debug=(bDrawDebug=true,bDraw2D=false,bDrawLocalAxes=false,DebugScale=25,DebugSegments=20,DebugThickness=0.25)
    DriverAxis=(X=-1,Y=0,Z=0)
    DriverItem=(Type=Bone,Name="thigh_l")
    FalloffRegionScaleFactors=(PositiveWidth=1,NegativeWidth=1,PositiveHeight=0.5,NegativeHeight=0.5)
    FalloffSize=0.5
    FlipHeightScaling=false
    FlipWidthScaling=false
    OptionalParentItem=(Type=Bone,Name="None")
    RotationOffset=(X=107.3557,Y=-2.9147,Z=-24.8654)
    ExecutePin=@RigVMFunction_Sequence_1.B
   )
   > &RigUnit_ModifyTransforms(
    ItemToModify=((Item=(Type=Bone,Name="dyn_skirt_d_MSTR"),Transform=(Rotation=(X=0.0435,Y=0.1737,Z=-0.2113,W=0.9609),Translation=(X=4.1102,Y=0.6119,Z=7.7711),Scale3D=(X=1,Y=1,Z=1))))
    WeightMinimum=0
    Weight=@n2.OutputParam
   )
   > &RigUnit_ModifyTransforms(
    ItemToModify=((Item=(Type=Bone,Name="dyn_skirt_h_MSTR"),Transform=(Rotation=(X=0,Y=0,Z=-0.1279,W=0.9918),Translation=(X=6.8112,Y=-3.9859,Z=7.3331),Scale3D=(X=1,Y=1,Z=1))))
    WeightMinimum=0
    Weight=@n2.OutputParam
   )
   > &RigUnit_ModifyTransforms(
    ItemToModify=((Item=(Type=Bone,Name="dyn_skirt_c_MSTR"),Transform=(Rotation=(X=-0.0403,Y=-0.0727,Z=-0.3968,W=0.9141),Translation=(X=-0.2591,Y=0,Z=1.6815),Scale3D=(X=1,Y=1,Z=1))))
    WeightMinimum=0
    Weight=@n2.OutputParam
   )
   &RigUnit_ParentConstraint(
    AdvancedSettings=(InterpolationType=Average,RotationOrderForFilter=XZY)
    Child=(Type=Bone,Name="dyn_waistchain_target")
    Filter=(TranslationFilter=(bX=True,bY=True,bZ=True),RotationFilter=(bX=false,bY=false,bZ=false),ScaleFilter=(bX=false,bY=false,bZ=false))
    Parents="((Item=(Type=Bone,Name=\"pelvis\"),Weight=0.4331),(Item=(Type=Bone,Name=\"thigh_r\"),Weight=0.0976),(Item=(Type=Bone,Name=\"thigh_twist_01_r\"),Weight=0.4099),(Item=(Type=Bone,Name=\"deform_groin_r\"),Weight=0.0309),(Item=(Type=Bone,Name=\"deform_glute_r\"),Weight=0 ...(truncated)"
    Weight=1
    ExecutePin=@RigVMFunction_Sequence_1.C
   )
   > &RigUnit_AimConstraintLocalSpaceOffset(
    Child=(Type=Bone,Name="dyn_waistchain_01")
    Filter=(bX=True,bY=True,bZ=True)
    Parents=((Item=(Type=Bone,Name="dyn_waistchain_target"),Weight=1))
    Weight=1
    WorldUp=(Target=(X=0,Y=0,Z=1),Kind=Location,Space=(Type=Bone,Name="dyn_waistchain_target"))
   )
   > n3 = &RigUnit_SphericalPoseReader(
    ActiveRegionScaleFactors=(PositiveWidth=0.8,NegativeWidth=0.3,PositiveHeight=0.3,NegativeHeight=0.3)
    ActiveRegionSize=0.4
    Debug=(bDrawDebug=true,bDraw2D=False,bDrawLocalAxes=False,DebugScale=25,DebugSegments=20,DebugThickness=0.25)
    DriverAxis=(X=1,Y=0,Z=0)
    DriverItem=(Type=Bone,Name="thigh_r")
    FalloffRegionScaleFactors=(PositiveWidth=1,NegativeWidth=1,PositiveHeight=0.5,NegativeHeight=0.5)
    FalloffSize=0.5
    FlipHeightScaling=false
    FlipWidthScaling=true
    OptionalParentItem=(Type=Bone,Name="None")
    RotationOffset=(X=-58.2527,Y=-6.7165,Z=-22.8064)
   )
   > &RigUnit_ModifyTransforms(
    ItemToModify=((Item=(Type=Bone,Name="dyn_waistchain_02"),Transform=(Rotation=(X=0.1433,Y=0,Z=0,W=0.9897),Translation=(X=2.7416,Y=1.4314,Z=17.0863),Scale3D=(X=1,Y=1,Z=1))))
    WeightMinimum=0
    Weight=@n3.OutputParam
   )
   &RigUnit_ParentConstraint(
    AdvancedSettings=(InterpolationType=Average,RotationOrderForFilter=XZY)
    Child=(Type=Bone,Name="dyn_skirt_d_MSTR")
    ExecutePin=()
    Filter=(TranslationFilter=(bX=false,bY=false,bZ=false),RotationFilter=(bX=true,bY=true,bZ=true),ScaleFilter=(bX=false,bY=false,bZ=false))
    Parents=((Item=(Type=Bone,Name="pelvis"),Weight=0),(Item=(Type=Bone,Name="thigh_l"),Weight=0),(Item=(Type=Bone,Name="thigh_twist_01_l"),Weight=1),(Item=(Type=Bone,Name="deform_groin_l"),Weight=0))
    Weight=1
   )
   &RigUnit_ParentConstraint(
    AdvancedSettings=(InterpolationType=Average,RotationOrderForFilter=XZY)
    Child=(Type=Bone,Name="dyn_skirt_h_MSTR")
    ExecutePin=()
    Filter=(TranslationFilter=(bX=false,bY=false,bZ=false),RotationFilter=(bX=true,bY=true,bZ=true),ScaleFilter=(bX=false,bY=false,bZ=false))
    Parents=((Item=(Type=Bone,Name="pelvis"),Weight=0),(Item=(Type=Bone,Name="thigh_l"),Weight=0),(Item=(Type=Bone,Name="thigh_twist_01_l"),Weight=1),(Item=(Type=Bone,Name="deform_groin_l"),Weight=0))
    Weight=1
   )
   &RigUnit_ParentConstraint(
    AdvancedSettings=(InterpolationType=Average,RotationOrderForFilter=XZY)
    Child=(Type=Bone,Name="dyn_skirt_c_MSTR")
    ExecutePin=()
    Filter=(TranslationFilter=(bX=false,bY=false,bZ=false),RotationFilter=(bX=true,bY=true,bZ=true),ScaleFilter=(bX=false,bY=false,bZ=false))
    Parents=((Item=(Type=Bone,Name="pelvis"),Weight=0),(Item=(Type=Bone,Name="thigh_twist_01_l"),Weight=1))
    Weight=1
   )
  ]}
]
hierarchy [
 root {loc=<0 0 0>, rot=<0 0 0 1>}
 pelvis {parent=root, loc=<0 -0.5301 98.0571>, rot=<0.0127 0.707 -0.0127 -0.707>}
 spine_01 {parent=pelvis, loc=<6.1208 -0.9328 0>, rot=<0 0 0.034 -0.9994>}
 spine_02 {parent=spine_01, loc=<6.5996 -0.2095 0>, rot=<0 0 0.0238 -0.9997>}
 spine_03 {parent=spine_02, loc=<6.4914 -0.1237 0>, rot=<0 0 0.0149 -0.9999>}
 spine_04 {parent=spine_03, loc=<6.9598 0.0794 0>, rot=<0 0 -0.0132 -0.9999>}
 spine_05 {parent=spine_04, loc=<6.9987 0 0>, rot=<0 0 -0.0015 -1>}
 attach_cape {parent=spine_05, loc=<9.3216 -8.4501 0>, rot=<0.0616 -0.7044 -0.0616 -0.7044>}
 clavicle_l {parent=spine_05, loc=<5.4831 -0.1053 -3.2946>, rot=<0.0465 -0.7289 -0.009 -0.6829>}
 upperarm_l {parent=clavicle_l, loc=<9.8658 -1.2136 -0.4105>, rot=<0.0164 -0.3245 -0.0269 -0.9454>}
 lowerarm_l {parent=upperarm_l, loc=<26.465 -0.8345 0.259>, rot=<0.0036 0.0161 -0.2298 -0.9731>}
 hand_l {parent=lowerarm_l, loc=<22.7355 -0.3203 -0.6105>, rot=<-0.6784 -0.0223 0.0735 -0.7307>}
 thumb_01_l {parent=hand_l, loc=<1.3118 -0.3191 -2.459>, rot=<0.6546 -0.4042 0.0664 -0.6354>}
 thumb_02_l {parent=thumb_01_l, loc=<5.9336 -0.0448 0.0401>, rot=<-0.0037 0.0213 0.298 -0.9543>}
 thumb_03_l {parent=thumb_02_l, loc=<2.2223 0.3682 -0.0907>, rot=<0.0118 -0.013 0.0066 -0.9998>}
 index_metacarpal_l {parent=hand_l, loc=<1.6535 0.0183 -2.96>, rot=<0.0158 -0.0124 -0.0173 -0.9996>}
 index_01_l {parent=index_metacarpal_l, loc=<8.3008 0.0357 -0.0524>, rot=<-0.0044 0.0485 0.0924 -0.9945>}
 index_02_l {parent=index_01_l, loc=<3.3587 -0.3004 0.0452>, rot=<-0.002 0.0027 0.3225 -0.9466>}
 index_03_l {parent=index_02_l, loc=<2.5578 -0.123 -0.0252>, rot=<-0.002 0.0067 0.2558 -0.9667>}
 middle_metacarpal_l {parent=hand_l, loc=<2.1856 -0.1725 -0.5984>, rot=<0.006 0.0168 -0.0329 -0.9993>}
 middle_01_l {parent=middle_metacarpal_l, loc=<7.3275 -0.2447 0.0129>, rot=<-0.0208 0.0042 0.0858 -0.9961>}
 middle_02_l {parent=middle_01_l, loc=<4.0123 0.1401 0.1175>, rot=<-0.0218 0.0258 0.3512 -0.9357>}
 middle_03_l {parent=middle_02_l, loc=<2.6077 0.0534 0.1373>, rot=<0.0082 0.0188 0.1708 -0.9851>}
 ring_metacarpal_l {parent=hand_l, loc=<2.2074 -0.3625 1.3638>, rot=<-0.0284 0.0722 -0.0458 -0.9959>}
 ring_01_l {parent=ring_metacarpal_l, loc=<7.0295 -0.6197 -0.1446>, rot=<-0.0489 -0.0252 0.1892 -0.9804>}
 ring_02_l {parent=ring_01_l, loc=<3.6174 0.4647 -0.1027>, rot=<0 0 0.2806 -0.9598>}
 ring_03_l {parent=ring_02_l, loc=<2.4152 0.114 0.0199>, rot=<0 0 0.1189 -0.9929>}
 pinky_metacarpal_l {parent=hand_l, loc=<1.7616 -0.9753 3.3814>, rot=<-0.0397 0.0814 -0.069 -0.9935>}
 pinky_01_l {parent=pinky_metacarpal_l, loc=<7.1709 -0.4831 -0.1094>, rot=<-0.0654 -0.0035 0.2242 -0.9723>}
 pinky_02_l {parent=pinky_01_l, loc=<2.8274 0.0893 -0.1043>, rot=<0.0037 0.0011 0.2854 -0.9584>}
 pinky_03_l {parent=pinky_02_l, loc=<2.2046 0.0307 -0.0078>, rot=<-0.0002 0.0098 0.1445 -0.9895>}
 deform_wrist_dn_base_l {parent=hand_l, loc=<0 -1.5326 0>, rot=<0 0 0 -1>}
 deform_wrist_dn_l {parent=deform_wrist_dn_base_l, loc=<0 0 0>, rot=<0 0 0 1>}
 deform_wrist_up_base_l {parent=hand_l, loc=<0 2.3009 0>, rot=<0 0 0 -1>}
 deform_wrist_up_l {parent=deform_wrist_up_base_l, loc=<0 0 0>, rot=<0 0 0 1>}
 lowerarm_twist_02_l {parent=lowerarm_l, loc=<10.6707 0 0>, rot=<0 0 0 1>}
 lowerarm_twist_01_l {parent=lowerarm_l, loc=<18.316 0 0>, rot=<0 0 0 1>}
 deform_elbow_in_base_l {parent=lowerarm_l, loc=<0.1834 3.6062 -0.2572>, rot=<0 0 0 -1>}
 deform_elbow_in_l {parent=deform_elbow_in_base_l, loc=<0 0 0>, rot=<0 0 0 1>}
 deform_elbow_out_base_l {parent=lowerarm_l, loc=<0 -4 0>, rot=<0 0 0 1>}
 deform_elbow_out_l {parent=deform_elbow_out_base_l, loc=<0 0 0>, rot=<0 0 0 1>}
 upperarm_twist_01_l {parent=upperarm_l, loc=<7.2135 -0.1134 0>, rot=<0 0 0 1>}
 deform_upperarm_bk_base_l {parent=upperarm_twist_01_l, loc=<-6.4053 -4.3746 0.314>, rot=<0 0 0 -1>}
 deform_upperarm_bk_l {parent=deform_upperarm_bk_base_l, loc=<0 0 0>, rot=<0 0 0 1>}
 deform_upperarm_fr_base_l {parent=upperarm_twist_01_l, loc=<-5.6312 5.092 0.3354>, rot=<0 0 0 -1>}
 deform_upperarm_fr_l {parent=deform_upperarm_fr_base_l, loc=<0 0 0>, rot=<0 0 0 1>}
 deform_upperarm_up_base_l {parent=upperarm_twist_01_l, loc=<-8.891 0.4825 3.9517>, rot=<0 0 0 -1>}
 deform_upperarm_up_l {parent=deform_upperarm_up_base_l, loc=<0 0 0>, rot=<0 0 0 1>}
 dyn_shoulderpad_l_MSTR {parent=upperarm_twist_01_l, loc=<-4.3215 0.1263 2.3667>, rot=<0 0 0 1>}
 dyn_shoulderpad_l {parent=dyn_shoulderpad_l_MSTR, loc=<0 0 1.0108>, rot=<0 0 0 1>}
 upperarm_twist_02_l {parent=upperarm_l, loc=<19.7998 -0.1134 0>, rot=<0 0 0 1>}
 attach_backpack {parent=spine_05, loc=<0 -8.35 0>, rot=<-0.0093 -0.707 0.0093 -0.707>}
 neck_01 {parent=spine_05, loc=<11.3138 0.1529 0>, rot=<0 0 -0.2003 -0.9797>}
 neck_02 {parent=neck_01, loc=<4.5627 0 0>, rot=<0 0 0.0504 -0.9987>}
 head {parent=neck_02, loc=<5.3914 0 0>, rot=<0 0 0.115 -0.9934>}
 clavicle_r {parent=spine_05, loc=<5.4831 -0.1053 3.2946>, rot=<0.7289 0.0465 -0.6829 0.009>}
 upperarm_r {parent=clavicle_r, loc=<-9.8658 1.2136 0.4104>, rot=<0.0164 -0.3245 -0.0269 -0.9454>}
 lowerarm_r {parent=upperarm_r, loc=<-26.4651 0.8345 -0.259>, rot=<0.0036 0.0161 -0.2298 -0.9731>}
 hand_r {parent=lowerarm_r, loc=<-22.7356 0.3203 0.6106>, rot=<-0.6784 -0.0223 0.0735 -0.7307>}
 thumb_01_r {parent=hand_r, loc=<-1.3119 0.3192 2.459>, rot=<0.6546 -0.4042 0.0664 -0.6354>}
 thumb_02_r {parent=thumb_01_r, loc=<-5.9336 0.0448 -0.0401>, rot=<-0.0037 0.0213 0.298 -0.9543>}
 thumb_03_r {parent=thumb_02_r, loc=<-2.2223 -0.3682 0.0907>, rot=<0.0118 -0.013 0.0066 -0.9998>}
 index_metacarpal_r {parent=hand_r, loc=<-1.6535 -0.0183 2.96>, rot=<0.0158 -0.0124 -0.0173 -0.9996>}
 index_01_r {parent=index_metacarpal_r, loc=<-8.3008 -0.0357 0.0524>, rot=<-0.0044 0.0485 0.0924 -0.9945>}
 index_02_r {parent=index_01_r, loc=<-3.3587 0.3004 -0.0452>, rot=<-0.002 0.0027 0.3225 -0.9466>}
 index_03_r {parent=index_02_r, loc=<-2.5578 0.123 0.0252>, rot=<-0.002 0.0067 0.2558 -0.9667>}
 middle_metacarpal_r {parent=hand_r, loc=<-2.1856 0.1726 0.5984>, rot=<0.006 0.0168 -0.0329 -0.9993>}
 middle_01_r {parent=middle_metacarpal_r, loc=<-7.3275 0.2447 -0.0129>, rot=<-0.0208 0.0042 0.0858 -0.9961>}
 middle_02_r {parent=middle_01_r, loc=<-4.0123 -0.1401 -0.1175>, rot=<-0.0218 0.0258 0.3512 -0.9357>}
 middle_03_r {parent=middle_02_r, loc=<-2.6077 -0.0534 -0.1373>, rot=<0.0082 0.0188 0.1708 -0.9851>}
 ring_metacarpal_r {parent=hand_r, loc=<-2.2073 0.3627 -1.3638>, rot=<-0.0284 0.0722 -0.0458 -0.9959>}
 ring_01_r {parent=ring_metacarpal_r, loc=<-7.0295 0.6197 0.1446>, rot=<-0.0489 -0.0252 0.1892 -0.9804>}
 ring_02_r {parent=ring_01_r, loc=<-3.6174 -0.4647 0.1027>, rot=<0 0 0.2806 -0.9598>}
 ring_03_r {parent=ring_02_r, loc=<-2.4152 -0.114 -0.0199>, rot=<0 0 0.1189 -0.9929>}
 pinky_metacarpal_r {parent=hand_r, loc=<-1.7617 0.9753 -3.3814>, rot=<-0.0397 0.0814 -0.069 -0.9935>}
 pinky_01_r {parent=pinky_metacarpal_r, loc=<-7.1709 0.4831 0.1094>, rot=<-0.0654 -0.0035 0.2242 -0.9723>}
 pinky_02_r {parent=pinky_01_r, loc=<-2.8274 -0.0893 0.1043>, rot=<0.0037 0.0011 0.2854 -0.9584>}
 pinky_03_r {parent=pinky_02_r, loc=<-2.2046 -0.0307 0.0078>, rot=<-0.0002 0.0098 0.1445 -0.9895>}
 deform_wrist_dn_base_r {parent=hand_r, loc=<-0.0001 1.5328 0>, rot=<0 0 0 -1>}
 deform_wrist_dn_r {parent=deform_wrist_dn_base_r, loc=<0 0 0>, rot=<0 0 0 1>}
 deform_wrist_up_base_r {parent=hand_r, loc=<-0.0002 -2.3007 -0.0001>, rot=<0 0 0 -1>}
 deform_wrist_up_r {parent=deform_wrist_up_base_r, loc=<0 0 0>, rot=<0 0 0 1>}
 lowerarm_twist_02_r {parent=lowerarm_r, loc=<-10.6707 0 0>, rot=<0 0 0 1>}
 lowerarm_twist_01_r {parent=lowerarm_r, loc=<-18.3161 0 0>, rot=<0 0 0 1>}
 deform_elbow_in_base_r {parent=lowerarm_r, loc=<-0.1834 -3.6063 0.2572>, rot=<0 0 0 -1>}
 deform_elbow_in_r {parent=deform_elbow_in_base_r, loc=<0 0 0>, rot=<0 0 0 1>}
 deform_elbow_out_base_r {parent=lowerarm_r, loc=<0 4 -0.0001>, rot=<0 0 0 1>}
 deform_elbow_out_r {parent=deform_elbow_out_base_r, loc=<0 0 0>, rot=<0 0 0 1>}
 upperarm_twist_01_r {parent=upperarm_r, loc=<-7.2135 0.1134 0>, rot=<0 0 0 1>}
 deform_upperarm_bk_base_r {parent=upperarm_twist_01_r, loc=<6.4052 4.3746 -0.314>, rot=<0 0 0 -1>}
 deform_upperarm_bk_r {parent=deform_upperarm_bk_base_r, loc=<0 0 0>, rot=<0 0 0 1>}
 deform_upperarm_fr_base_r {parent=upperarm_twist_01_r, loc=<5.6311 -5.092 -0.3354>, rot=<0 0 0 -1>}
 deform_upperarm_fr_r {parent=deform_upperarm_fr_base_r, loc=<0 0 0>, rot=<0 0 0 1>}
 deform_upperarm_up_base_r {parent=upperarm_twist_01_r, loc=<8.891 -0.4825 -3.9517>, rot=<0 0 0 -1>}
 deform_upperarm_up_r {parent=deform_upperarm_up_base_r, loc=<0 0 0>, rot=<0 0 0 1>}
 upperarm_twist_02_r {parent=upperarm_r, loc=<-19.7998 0.1134 0>, rot=<0 0 0 1>}
 deform_pec_aim_base_l {parent=spine_04, loc=<3.2456 9.3309 -1.2237>, rot=<0.0603 -0.4033 0.3318 -0.8507>}
 deform_pec_aim_l {parent=deform_pec_aim_base_l, loc=<0 0 0>, rot=<0 0 0 1>}
 deform_pec_base_l {parent=deform_pec_aim_l, loc=<9.3665 0 0>, rot=<0 0 0 1>}
 deform_pec_l {parent=deform_pec_base_l, loc=<0 0 0>, rot=<0 0 0 1>}
 deform_pec_aim_base_r {parent=spine_04, loc=<3.2452 9.3309 1.2237>, rot=<-0.4033 -0.0603 0.8507 0.3318>}
 deform_pec_aim_r {parent=deform_pec_aim_base_r, loc=<0 0 0>, rot=<0 0 0 -1>}
 deform_pec_base_r {parent=deform_pec_aim_r, loc=<-9.3667 -0.0003 -0.0004>, rot=<0 0 0 -1>}
 deform_pec_r {parent=deform_pec_base_r, loc=<0 0 0>, rot=<0 0 0 -1>}
 dyn_torsochain_01 {parent=spine_04, loc=<7.8268 0.3092 4.5593>, rot=<-0.0827 -0.9251 0.0106 -0.3704>}
 dyn_shirtbott_a_MSTR {parent=spine_02, loc=<-0.9434 10.3519 -0.1803>, rot=<0 0.7071 0.7071 0>}
 dyn_shirtbott_a_01 {parent=dyn_shirtbott_a_MSTR, loc=<0.7009 0 0>, rot=<0 0 0 1>}
 dyn_shirtbott_b_MSTR {parent=spine_02, loc=<3.1116 9.3127 5.2444>, rot=<0 0.5896 0.8077 0>}
 dyn_shirtbott_b_01 {parent=dyn_shirtbott_b_MSTR, loc=<0.7009 0 0>, rot=<0 0 0 1>}
 dyn_shirtbott_c_MSTR {parent=spine_02, loc=<6.5221 3.2828 10.7256>, rot=<0.0023 0.0822 0.9962 -0.0283>}
 dyn_shirtbott_c_01 {parent=dyn_shirtbott_c_MSTR, loc=<0.7009 0 0>, rot=<0 0 0 1>}
 dyn_shirtbott_d_MSTR {parent=spine_02, loc=<6.5985 -4.4513 7.7643>, rot=<0.0214 -0.6349 0.7721 -0.0187>}
 dyn_shirtbott_d_01 {parent=dyn_shirtbott_d_MSTR, loc=<0.7009 0 0>, rot=<0 0 0 1>}
 dyn_shirtbott_e_MSTR {parent=spine_02, loc=<1.6209 -6.4211 0>, rot=<0 0.7071 -0.7071 0>}
 dyn_shirtbott_e_01 {parent=dyn_shirtbott_e_MSTR, loc=<0.7009 0 0>, rot=<0 0 0 1>}
 dyn_shirtbott_f_MSTR {parent=spine_02, loc=<3.2306 9.0286 -5.997>, rot=<-0.8468 0 0 0.5319>}
 dyn_shirtbott_f_01 {parent=dyn_shirtbott_f_MSTR, loc=<-0.7009 0 0>, rot=<0 0 0 1>}
 spine_03_2 {parent=spine_02, loc=<6.4914 -0.1237 0>, rot=<0 0 0.0149 -0.9999>}
 thigh_l {parent=pelvis, loc=<-1.0855 -0.2109 -8.7053>, rot=<0.0009 0.0355 0.0155 -0.9993>}
 calf_l {parent=thigh_l, loc=<-42.5284 -0.0019 -0.7297>, rot=<0.0006 -0.0161 -0.0391 -0.9991>}
 foot_l {parent=calf_l, loc=<-46.7401 -2.5014 -0.592>, rot=<0.0427 -0.0238 0.0353 -0.9982>}
 ball_l {parent=foot_l, loc=<-3.5658 10.8066 -0.1544>, rot=<0.0001 -0.0001 -0.7063 -0.7079>}
 calf_twist_02_l {parent=calf_l, loc=<-19.1546 -1.7943 -0.2638>, rot=<0.0237 0.0095 -0.0035 -0.9997>}
 dyn_calf_flap_top_MSTR {parent=calf_twist_02_l, loc=<-0.8647 -0.2739 -6.1647>, rot=<0.326 -0.6453 0.3358 -0.6038>}
 dyn_calf_flap_top_01 {parent=dyn_calf_flap_top_MSTR, loc=<0.6438 0 0>, rot=<0 0 0 1>}
 dyn_calf_flap_bott_MSTR {parent=calf_twist_02_l, loc=<-6.9683 -0.9677 -5.7707>, rot=<0.4435 -0.6154 0.387 -0.5242>}
 dyn_calf_flap_bott_01 {parent=dyn_calf_flap_bott_MSTR, loc=<0.6438 0 0>, rot=<0 0 0 1>}
 calf_twist_01_l {parent=calf_l, loc=<-35.8483 -1.9197 -0.5794>, rot=<0.0237 0.0095 -0.0035 -0.9997>}
 deform_knee_in_base_l {parent=calf_l, loc=<0.8031 -5.0231 0.2456>, rot=<0 0 0 -1>}
 deform_knee_in_l {parent=deform_knee_in_base_l, loc=<0 0 0>, rot=<0 0 0 1>}
 deform_knee_out_base_l {parent=calf_l, loc=<0 4 0>, rot=<0 0 0 -1>}
 deform_knee_out_l {parent=deform_knee_out_base_l, loc=<0 0 0>, rot=<0 0 0 1>}
 dyn_kneepad_l {parent=calf_l, loc=<0.8548 5.5011 0.1524>, rot=<0 0 0 -1>}
 thigh_twist_01_l {parent=thigh_l, loc=<-27.952 0 0>, rot=<0.0013 0 0 -1>}
 dyn_thigh_rip_l_MSTR {parent=thigh_twist_01_l, loc=<3.3503 3.0047 5.6191>, rot=<0.531 0.348 0.7709 -0.0513>}
 dyn_thigh_rip_l {parent=dyn_thigh_rip_l_MSTR, loc=<0.5846 0 0>, rot=<0 0 0 1>}
 dyn_skirt_d_MSTR {parent=thigh_twist_01_l, loc=<30.2732 5.343 -3.2067>, rot=<-0.0173 -0.8644 -0.4912 -0.1066>}
 dyn_skirt_d_01 {parent=dyn_skirt_d_MSTR, loc=<0.6762 0 0>, rot=<0 0 0 1>}
 dyn_skirt_d_02 {parent=dyn_skirt_d_01, loc=<7.1805 0 0.2234>, rot=<0 0.0276 0 0.9996>}
 dyn_skirt_d_03 {parent=dyn_skirt_d_02, loc=<8.6424 0 0>, rot=<0 0.0496 0 0.9988>}
 dyn_skirt_d_04 {parent=dyn_skirt_d_03, loc=<8.6626 0 0>, rot=<0 0 0 1>}
 dyn_skirt_c_MSTR {parent=thigh_twist_01_l, loc=<31.3858 -0.7263 -5.2327>, rot=<0.0434 -0.9797 0.1653 -0.1048>}
 dyn_skirt_c_01 {parent=dyn_skirt_c_MSTR, loc=<0.7729 0 0>, rot=<0 0 0 1>}
 dyn_skirt_c_02 {parent=dyn_skirt_c_01, loc=<10.4755 0 0>, rot=<-0.0313 -0.0803 -0.0025 -0.9963>}
 dyn_skirt_c_03 {parent=dyn_skirt_c_02, loc=<11.9449 0 0>, rot=<0 -0.0523 0 0.9986>}
 dyn_skirt_c_04 {parent=dyn_skirt_c_03, loc=<11.9651 0 0>, rot=<0 -0.041 0 0.9992>}
 dyn_skirt_c_05 {parent=dyn_skirt_c_04, loc=<12.9248 0 0>, rot=<0 0 0 1>}
 dyn_skirt_c_06 {parent=dyn_skirt_c_05, loc=<10.8743 0 0>, rot=<0 0 0 1>}
 dyn_skirt_c_07 {parent=dyn_skirt_c_06, loc=<8.4504 0 0>, rot=<0 0 0 1>}
 dyn_skirt_h_MSTR {parent=thigh_twist_01_l, loc=<29.6473 7.9972 2.8485>, rot=<-0.0508 -0.7632 -0.6424 -0.0478>}
 dyn_skirt_h_01 {parent=dyn_skirt_h_MSTR, loc=<0.6762 0 0>, rot=<0 0 0 1>}
 dyn_skirt_h_02 {parent=dyn_skirt_h_01, loc=<7.2022 0 0.2219>, rot=<0 0.0338 0 0.9994>}
 deform_groin_base_l {parent=thigh_l, loc=<0.4801 5.6776 -0.2486>, rot=<0 0 0 -1>}
 deform_groin_l {parent=deform_groin_base_l, loc=<0 0 0>, rot=<0 0 0 1>}
 thigh_r {parent=pelvis, loc=<-1.0855 -0.2109 8.7053>, rot=<0.0355 -0.0009 0.9993 0.0155>}
 calf_r {parent=thigh_r, loc=<42.5284 0.0019 0.7297>, rot=<0.0006 -0.0161 -0.0391 -0.9991>}
 foot_r {parent=calf_r, loc=<46.7401 2.5014 0.592>, rot=<0.0427 -0.0238 0.0353 -0.9982>}
 ball_r {parent=foot_r, loc=<3.5658 -10.8066 0.1544>, rot=<0.0001 -0.0001 -0.7063 -0.7079>}
 calf_twist_02_r {parent=calf_r, loc=<19.1546 1.7943 0.2638>, rot=<0.0237 0.0095 -0.0035 -0.9997>}
 calf_twist_01_r {parent=calf_r, loc=<35.8483 1.9197 0.5794>, rot=<0.0237 0.0095 -0.0035 -0.9997>}
 deform_knee_in_base_r {parent=calf_r, loc=<-0.8031 5.0231 -0.2457>, rot=<0 0 0 -1>}
 deform_knee_in_r {parent=deform_knee_in_base_r, loc=<0 0 0>, rot=<0 0 0 1>}
 deform_knee_out_base_r {parent=calf_r, loc=<0 -4 0>, rot=<0 0 0 -1>}
 deform_knee_out_r {parent=deform_knee_out_base_r, loc=<0 0 0>, rot=<0 0 0 1>}
 thigh_twist_01_r {parent=thigh_r, loc=<27.952 0 0>, rot=<0.0013 0 0 -1>}
 dyn_skirt_e_MSTR {parent=thigh_twist_01_r, loc=<-26.7154 -8.5407 -3.2877>, rot=<0.6213 -0.0319 -0.0615 0.7805>}
 dyn_skirt_e_01 {parent=dyn_skirt_e_MSTR, loc=<0.7971 0 0>, rot=<0 0 0 1>}
 dyn_skirt_g_MSTR {parent=thigh_twist_01_r, loc=<-25.3617 -3.6271 6.8008>, rot=<0.1995 -0.0808 -0.053 0.9751>}
 dyn_skirt_g_01 {parent=dyn_skirt_g_MSTR, loc=<0.5657 0 0>, rot=<0 0 0 1>}
 deform_groin_base_r {parent=thigh_r, loc=<-0.4801 -5.6776 0.2486>, rot=<0 0 0 -1>}
 deform_groin_r {parent=deform_groin_base_r, loc=<0 0 0>, rot=<0 0 0 1>}
 deform_glute_base_l {parent=pelvis, loc=<-7.3466 -9.0213 -9.1245>, rot=<0.0172 0.031 -0.4477 -0.8935>}
 deform_glute_l {parent=deform_glute_base_l, loc=<0 0 0>, rot=<0 0 0 1>}
 deform_glute_base_r {parent=pelvis, loc=<-7.3464 -9.0213 9.1245>, rot=<0.031 -0.0172 0.8935 -0.4477>}
 deform_glute_r {parent=deform_glute_base_r, loc=<0 0 0>, rot=<0 0 0 1>}
 dyn_skirt_a_01 {parent=pelvis, loc=<1.3832 -11.2665 0>, rot=<-0.0695 0.7037 -0.7037 0.0695>}
 dyn_skirt_a_02 {parent=dyn_skirt_a_01, loc=<10.4755 0 0>, rot=<0 -0.0803 0 -0.9968>}
 dyn_skirt_a_03 {parent=dyn_skirt_a_02, loc=<11.9449 0 0>, rot=<0 0 0 1>}
 dyn_skirt_b_01 {parent=pelvis, loc=<1.3832 -8.7184 -8.6654>, rot=<0.0139 -0.883 0.4563 -0.1097>}
 dyn_skirt_b_02 {parent=dyn_skirt_b_01, loc=<10.4755 0 0>, rot=<-0.0313 -0.0803 -0.0025 -0.9963>}
 dyn_skirt_b_03 {parent=dyn_skirt_b_02, loc=<11.9449 0 0>, rot=<0 -0.0523 0 0.9986>}
 dyn_skirt_b_04 {parent=dyn_skirt_b_03, loc=<11.9651 0 0>, rot=<0 -0.041 0 0.9992>}
 dyn_skirt_b_05 {parent=dyn_skirt_b_04, loc=<12.9248 0 0>, rot=<0 0 0 1>}
 dyn_skirt_b_06 {parent=dyn_skirt_b_05, loc=<10.8743 0 0>, rot=<0 0 0 1>}
 dyn_skirt_f_01 {parent=pelvis, loc=<-0.8911 -9.8367 9.1538>, rot=<-0.0674 0.6825 -0.7242 0.0716>}
 dyn_skirt_f_02 {parent=dyn_skirt_f_01, loc=<7.9515 0 0>, rot=<0 0 0 1>}
 dyn_waistchain_01 {parent=pelvis, loc=<5.4447 -1.4993 -7.8943>, rot=<0.592 0.1323 0.7942 0.0365>}
 dyn_waistchain_02 {parent=dyn_waistchain_01, loc=<19.86 -7.5192 -0.6711>, rot=<-0.0277 -0.0572 -0.4352 0.8981>}
 dyn_waistchain_target {parent=pelvis, loc=<-3.1681 -0.7665 15.9962>, rot=<-0.634 -0.002 -0.7727 0.0321>}
]
