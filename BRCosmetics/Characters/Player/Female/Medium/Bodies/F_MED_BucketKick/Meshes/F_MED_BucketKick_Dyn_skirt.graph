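# Physics asset for the dynamic skirt of the F_MED_BucketKick female character body, with capsule bodies on six skirt bone chains plus pelvis, thighs and calves.
# Chains a, b and f constrain directly to the pelvis, while c, d and h attach through tiny MSTR anchor bodies.
# Skirt links have tight twist limits, with one swing axis zeroed along chains c and d for hinge-like bending.
# Leg constraints and the c and h anchors are free in rotation.
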PhysicsAsset /BRCosmetics/Characters/Player/Female/Medium/Bodies/F_MED_BucketKick/Meshes/F_MED_BucketKick_Dyn_skirt
bodies {
 calf_l [{shapeType=Capsule, length=33.9282, radius=8.1686, shapeName=thigh_l_capsule, center=<-21.9895 -2.9093 -1.2089>, rotation=<rot 90 0 0>}]
 calf_r [{shapeType=Capsule, length=33.9282, radius=8.1686, shapeName=thigh_l_capsule, center=<21.9895 2.9093 1.2089>, rotation=<rot 90 0 180>}]
 dyn_skirt_a_01 [{shapeType=Capsule, length=6.9471, radius=6.715, shapeName=dyn_skirt_a_01_capsule, center=<5.0053 0 4.8728>, rotation=<rot 90 0 0>}]
 dyn_skirt_a_02 [{shapeType=Capsule, length=6.9471, radius=6.715, shapeName=dyn_skirt_a_01_capsule, center=<5.0053 0 4.8728>, rotation=<rot 90 0 0>}]
 dyn_skirt_a_03 [{shapeType=Capsule, length=6.9471, radius=6.715, shapeName=dyn_skirt_a_01_capsule, center=<5.0053 0 4.8728>, rotation=<rot 90 0 0>}]
 dyn_skirt_b_01 [{shapeType=Capsule, length=6.9471, radius=6.715, shapeName=dyn_skirt_a_01_capsule, center=<5.0053 0 6.9259>, rotation=<rot 90 0 0>}]
 dyn_skirt_b_02 [{shapeType=Capsule, length=6.9471, radius=6.715, shapeName=dyn_skirt_a_01_capsule, center=<5.0053 0 5.9829>, rotation=<rot 90 0 0>}]
 dyn_skirt_b_03 [{shapeType=Capsule, length=6.9471, radius=6.715, shapeName=dyn_skirt_a_01_capsule, center=<5.0053 0 5.9829>, rotation=<rot 90 0 0>}]
 dyn_skirt_b_04 [{shapeType=Capsule, length=6.9471, radius=6.715, shapeName=dyn_skirt_a_01_capsule, center=<5.0053 0 5.9829>, rotation=<rot 90 0 0>}]
 dyn_skirt_b_05 [{shapeType=Capsule, length=6.9471, radius=6.715, shapeName=dyn_skirt_a_01_capsule, center=<5.0053 0 5.9829>, rotation=<rot 90 0 0>}]
 dyn_skirt_b_06 [{shapeType=Capsule, length=6.9471, radius=6.715, shapeName=dyn_skirt_a_01_capsule, center=<5.0053 0 5.9829>, rotation=<rot 90 0 0>}]
 dyn_skirt_c_01 [{shapeType=Capsule, length=6.9471, radius=6.715, shapeName=dyn_skirt_a_01_capsule, center=<5.0053 0 5.9829>, rotation=<rot 90 0 0>}]
 dyn_skirt_c_02 [{shapeType=Capsule, length=6.9471, radius=6.715, shapeName=dyn_skirt_a_01_capsule, center=<5.0053 0 5.9829>, rotation=<rot 90 0 0>}]
 dyn_skirt_c_03 [{shapeType=Capsule, length=6.9471, radius=6.715, shapeName=dyn_skirt_a_01_capsule, center=<5.0053 0 5.9829>, rotation=<rot 90 0 0>}]
 dyn_skirt_c_04 [{shapeType=Capsule, length=6.9471, radius=6.715, shapeName=dyn_skirt_a_01_capsule, center=<5.0053 0 5.9829>, rotation=<rot 90 0 0>}]
 dyn_skirt_c_05 [{shapeType=Capsule, length=6.9471, radius=6.715, shapeName=dyn_skirt_a_01_capsule, center=<5.0053 0 5.9829>, rotation=<rot 90 0 0>}]
 dyn_skirt_c_06 [{shapeType=Capsule, length=6.9471, radius=6.715, shapeName=dyn_skirt_a_01_capsule, center=<5.0053 0 5.9829>, rotation=<rot 90 0 0>}]
 dyn_skirt_c_07 [{shapeType=Capsule, length=6.9471, radius=6.715, shapeName=dyn_skirt_a_01_capsule, center=<5.0053 0 5.9829>, rotation=<rot 90 0 0>}]
 dyn_skirt_c_MSTR [{shapeType=Capsule, length=0.5, radius=0.5, shapeName=dyn_skirt_a_01_capsule}]
 dyn_skirt_d_01 [{shapeType=Capsule, length=6.9471, radius=6.715, shapeName=dyn_skirt_a_01_capsule, center=<5.0053 0 7.4008>, rotation=<rot 90 0 0>}]
 dyn_skirt_d_02 [{shapeType=Capsule, length=6.9471, radius=6.715, shapeName=dyn_skirt_a_01_capsule, center=<5.0053 0 6.9555>, rotation=<rot 90 0 0>}]
 dyn_skirt_d_03 [{shapeType=Capsule, length=6.9471, radius=6.715, shapeName=dyn_skirt_a_01_capsule, center=<5.0053 0 6.2734>, rotation=<rot 90 0 0>}]
 dyn_skirt_d_04 [{shapeType=Capsule, length=6.9471, radius=6.715, shapeName=dyn_skirt_a_01_capsule, center=<5.0053 0 6.2734>, rotation=<rot 90 0 0>}]
 dyn_skirt_d_MSTR [{shapeType=Capsule, length=0.5, radius=0.5, shapeName=dyn_skirt_a_01_capsule}]
 dyn_skirt_f_01 [{shapeType=Capsule, length=6.9471, radius=6.715, shapeName=dyn_skirt_a_01_capsule, center=<5.0053 0 5.9829>, rotation=<rot 90 0 0>}]
 dyn_skirt_f_02 [{shapeType=Capsule, length=14.9022, radius=6.715, shapeName=dyn_skirt_a_01_capsule, center=<7.0323 0 3.9102>, rotation=<rot 81.753 0 0>}]
 dyn_skirt_h_01 [{shapeType=Capsule, length=6.9471, radius=6.715, shapeName=dyn_skirt_a_01_capsule, center=<5.0053 0 6.5087>, rotation=<rot 90 0 0>}]
 dyn_skirt_h_02 [{shapeType=Capsule, length=6.9471, radius=6.715, shapeName=dyn_skirt_a_01_capsule, center=<5.0053 0 7.1529>, rotation=<rot 90 0 0>}]
 dyn_skirt_h_MSTR [{shapeType=Capsule, length=0.5, radius=0.5, shapeName=dyn_skirt_a_01_capsule}]
 pelvis [
  {shapeType=Capsule, length=15.8946, radius=8.8482, shapeName=pelvis_capsule, center=<-3.8012 -1.0548 2.5984>}
  {shapeType=Capsule, length=14.269, radius=10.4737, shapeName=pelvis_capsule, center=<-6.9699 -1.5501 2.5984>}
 ]
 thigh_l [{
   shapeType=Capsule
   length=32.0386
   radius=10.0582
   shapeName=thigh_l_capsule
   center=<-22.1596 -0.0982 2.2706>
   rotation=<rot 70.238 112.5004 114.0955>
  }]
 thigh_r [{shapeType=Capsule, length=33.9282, radius=8.1686, shapeName=thigh_l_capsule, center=<21.9895 0 -0.3805>, rotation=<rot 85.7624 179.9999 0>}]
}
constraints [
 {bone1Name=dyn_skirt_a_02, bone2Name=dyn_skirt_a_01, swing1LimitDegrees=45, swing2LimitDegrees=45, twistLimitDegrees=5}
 {bone1Name=dyn_skirt_a_03, bone2Name=dyn_skirt_a_02, swing1LimitDegrees=45, swing2LimitDegrees=45, twistLimitDegrees=5}
 {bone1Name=dyn_skirt_b_02, bone2Name=dyn_skirt_b_01, swing1LimitDegrees=45, swing2LimitDegrees=45, twistLimitDegrees=5}
 {bone1Name=dyn_skirt_b_03, bone2Name=dyn_skirt_b_02, swing1LimitDegrees=45, swing2LimitDegrees=45, twistLimitDegrees=5}
 {bone1Name=dyn_skirt_b_04, bone2Name=dyn_skirt_b_03, swing1LimitDegrees=45, swing2LimitDegrees=45, twistLimitDegrees=5}
 {bone1Name=dyn_skirt_b_05, bone2Name=dyn_skirt_b_04, swing1LimitDegrees=45, swing2LimitDegrees=45, twistLimitDegrees=5}
 {bone1Name=dyn_skirt_b_06, bone2Name=dyn_skirt_b_05, swing1LimitDegrees=45, swing2LimitDegrees=45, twistLimitDegrees=5}
 {bone1Name=dyn_skirt_c_02, bone2Name=dyn_skirt_c_01, swing1LimitDegrees=0, swing2LimitDegrees=45, twistLimitDegrees=5}
 {bone1Name=dyn_skirt_c_03, bone2Name=dyn_skirt_c_02, swing1LimitDegrees=0, swing2LimitDegrees=45, twistLimitDegrees=5}
 {bone1Name=dyn_skirt_c_04, bone2Name=dyn_skirt_c_03, swing1LimitDegrees=0, swing2LimitDegrees=45, twistLimitDegrees=5}
 {bone1Name=dyn_skirt_c_05, bone2Name=dyn_skirt_c_04, swing1LimitDegrees=0, swing2LimitDegrees=45, twistLimitDegrees=5}
 {bone1Name=dyn_skirt_c_06, bone2Name=dyn_skirt_c_05, swing1LimitDegrees=0, swing2LimitDegrees=45, twistLimitDegrees=5}
 {bone1Name=dyn_skirt_c_07, bone2Name=dyn_skirt_c_06, swing1LimitDegrees=0, swing2LimitDegrees=45, twistLimitDegrees=5}
 {bone1Name=dyn_skirt_d_01, bone2Name=dyn_skirt_d_MSTR, swing1LimitDegrees=0, swing2LimitDegrees=45, twistLimitDegrees=5}
 {bone1Name=dyn_skirt_d_02, bone2Name=dyn_skirt_d_01, swing1LimitDegrees=0, swing2LimitDegrees=45, twistLimitDegrees=5}
 {bone1Name=dyn_skirt_d_03, bone2Name=dyn_skirt_d_02, swing1LimitDegrees=0, swing2LimitDegrees=45, twistLimitDegrees=5}
 {bone1Name=dyn_skirt_d_04, bone2Name=dyn_skirt_d_03, swing1LimitDegrees=0, swing2LimitDegrees=45, twistLimitDegrees=5}
 {bone1Name=dyn_skirt_a_01, bone2Name=pelvis, swing1LimitDegrees=2, swing2LimitDegrees=45, twistLimitDegrees=5}
 {bone1Name=dyn_skirt_b_01, bone2Name=pelvis, swing1LimitDegrees=2, swing2LimitDegrees=45, twistLimitDegrees=5}
 {bone1Name=dyn_skirt_c_01, bone2Name=dyn_skirt_c_MSTR, swing1LimitDegrees=0, swing2LimitDegrees=45, twistLimitDegrees=5}
 {bone1Name=dyn_skirt_d_MSTR, bone2Name=pelvis, swing1LimitDegrees=45, swing2LimitDegrees=45, twistLimitDegrees=5}
 {
  bone1Name=thigh_l
  bone2Name=pelvis
  swing1LimitDegrees=45
  swing1Motion=Free
  swing2LimitDegrees=45
  swing2Motion=Free
  twistLimitDegrees=45
  twistMotion=Free
 }
 {bone1Name=dyn_skirt_f_01, bone2Name=pelvis, swing1LimitDegrees=2, swing2LimitDegrees=45, twistLimitDegrees=5}
 {bone1Name=dyn_skirt_f_02, bone2Name=dyn_skirt_f_01, swing1LimitDegrees=45, swing2LimitDegrees=45, twistLimitDegrees=5}
 {
  bone1Name=calf_l
  bone2Name=thigh_l
  swing1LimitDegrees=45
  swing1Motion=Free
  swing2LimitDegrees=45
  swing2Motion=Free
  twistLimitDegrees=45
  twistMotion=Free
 }
 {
  bone1Name=thigh_r
  bone2Name=pelvis
  swing1LimitDegrees=45
  swing1Motion=Free
  swing2LimitDegrees=45
  swing2Motion=Free
  twistLimitDegrees=45
  twistMotion=Free
 }
 {
  bone1Name=calf_r
  bone2Name=thigh_r
  swing1LimitDegrees=45
  swing1Motion=Free
  swing2LimitDegrees=45
  swing2Motion=Free
  twistLimitDegrees=45
  twistMotion=Free
 }
 {bone1Name=dyn_skirt_h_01, bone2Name=dyn_skirt_h_MSTR, swing1LimitDegrees=5, swing2LimitDegrees=45, twistLimitDegrees=5}
 {bone1Name=dyn_skirt_h_02, bone2Name=dyn_skirt_h_01, swing1LimitDegrees=5, swing2LimitDegrees=45, twistLimitDegrees=5}
 {
  bone1Name=dyn_skirt_h_MSTR
  bone2Name=pelvis
  swing1LimitDegrees=45
  swing1Motion=Free
  swing2LimitDegrees=45
  swing2Motion=Free
  twistLimitDegrees=45
  twistMotion=Free
 }
 {
  bone1Name=dyn_skirt_c_MSTR
  bone2Name=pelvis
  swing1LimitDegrees=45
  swing1Motion=Free
  swing2LimitDegrees=45
  swing2Motion=Free
  twistLimitDegrees=45
  twistMotion=Free
 }
]
