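# Physics asset for the F_MED_BucketKick female player body that, as its name indicates, supplies collision shapes for the character's cape.
# Capsules cover the pelvis, spine, neck and clavicles, with a sphere on the head and extra capsules on front and back dynamic skirt bones.
# Arm, leg and foot bodies carry no shapes.
# It defines no constraints, so it holds only colliders, with no joint setup.
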PhysicsAsset /BRCosmetics/Characters/Player/Female/Medium/Bodies/F_MED_BucketKick/Meshes/F_MED_BucketKick_CapeColliders
bodies {
 calf_l []
 calf_r []
 clavicle_l [{shapeType=Capsule, length=10.3434, radius=7.5466, center=<9.2095 -1.1898 -2.8765>, rotation=<rot 60.1545 2.4265 177.8344>}]
 clavicle_r [{shapeType=Capsule, length=10.3434, radius=7.5466, center=<-9.2095 1.1898 2.8765>, rotation=<rot 60.1545 2.4265 -2.1656>}]
 dyn_skirt_b_02 [{shapeType=Capsule, length=21.4599, radius=12.2285, shapeName=dyn_skirt_b_04_capsule, center=<9.5795 0 -10.0177>, rotation=<rot 90 0 0>}]
 dyn_skirt_b_04 [{
   shapeType=Capsule
   length=21.4599
   radius=12.2285
   shapeName=dyn_skirt_b_04_capsule
   center=<9.5795 0 -12.0548>
   rotation=<rot 81.8399 -0.0001 -0.0001>
  }]
 dyn_skirt_b_06 [{
   shapeType=Capsule
   length=21.4599
   radius=12.2285
   shapeName=dyn_skirt_b_04_capsule
   center=<0.7269 0 -11.8124>
   rotation=<rot 88.4313 -0.0006 -0.0006>
  }]
 dyn_skirt_f_02 [{shapeType=Capsule, length=21.4599, radius=12.2285, shapeName=dyn_skirt_b_04_capsule, center=<8.7446 0 -12.777>, rotation=<rot 77.6067 0 0>}]
 foot_l []
 foot_r []
 head [{shapeType=Sphere, radius=17.3125, center=<11.028 -0.174 0>}]
 lowerarm_l []
 lowerarm_r []
 neck_02 [{shapeType=Capsule, length=13.4375, radius=7.3175, center=<-0.2642 -0.3408 0>, rotation=<rot 90 -171.1193 -175.3978>}]
 pelvis [{shapeType=Capsule, length=6.4496, radius=13.1368, center=<-1.3827 -2.2221 0>}]
 spine_01 [{shapeType=Capsule, length=1.0591, radius=12.4995, center=<-1.7965 0.1884 0>}]
 spine_02 [
  {shapeType=Capsule, length=4.5109, radius=10.109, center=<-1.711 -0.1851 0.0001>}
  {shapeType=Capsule, length=31.1904, radius=9.7859, center=<2.5024 0.4992 3.9825>, rotation=<rot 90 0.0049 0.0049>}
  {shapeType=Capsule, length=31.6698, radius=9.3065, center=<2.4704 0.4991 -3.9825>, rotation=<rot 90 -0.0396 -0.0396>}
 ]
 spine_03 [{shapeType=Capsule, length=6.0932, radius=10.1453, center=<0.1445 0.2396 0.0001>}]
 spine_04 [{shapeType=Capsule, length=3.4107, radius=14.1325, center=<2.766 0.7765 0.0002>}]
 spine_05 [
  {shapeType=Capsule, length=2.7983, radius=14.9099, center=<-2.2795 1.8109 0>}
  {shapeType=Capsule, length=11.5567, radius=13.3178, center=<-0.2916 0.4452 0>}
 ]
 thigh_l []
 thigh_r []
 upperarm_l []
 upperarm_r []
}
constraints []
