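# Cloth collision physics asset for the F_MED_BucketKick female medium player body cosmetic, judging by its name and path.
# The only collision shape is a single capsule on the pelvis, while the left thigh, calf and foot bodies have no shapes.
# Constraints chain foot to calf to thigh to pelvis, with all swing and twist axes left free.
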PhysicsAsset /BRCosmetics/Characters/Player/Female/Medium/Bodies/F_MED_BucketKick/Meshes/F_MED_BucketKick_ClothCollision
bodies {
 calf_l []
 foot_l []
 pelvis [{shapeType=Capsule, length=7.3944, radius=9.7678, shapeName=pelvis_capsule, center=<-3.7268 -1.6608 0>}]
 thigh_l []
}
constraints [
 {
  bone1Name=calf_l
  bone2Name=thigh_l
  swing1LimitDegrees=45
  swing1Motion=Free
  swing2LimitDegrees=45
  swing2Motion=Free
  twistLimitDegrees=45
  twistMotion=Free
 }
 {
  bone1Name=foot_l
  bone2Name=calf_l
  swing1LimitDegrees=45
  swing1Motion=Free
  swing2LimitDegrees=45
  swing2Motion=Free
  twistLimitDegrees=45
  twistMotion=Free
 }
 {
  bone1Name=thigh_l
  bone2Name=pelvis
  swing1LimitDegrees=45
  swing1Motion=Free
  swing2LimitDegrees=45
  swing2Motion=Free
  twistLimitDegrees=45
  twistMotion=Free
 }
]
